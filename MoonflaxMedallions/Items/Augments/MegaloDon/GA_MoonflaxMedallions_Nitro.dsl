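; GA_MoonflaxMedallions_Nitro is a gameplay ability for a Nitro effect, filed under the MegaloDon augment of MoonflaxMedallions.
; On activation it runs the parent ability's activation, which is not shown here.
; It then registers Nitro analytics for the avatar's Fortnite player controller, tagged with the ability's NitroAnalyticsSource variable.
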(event Ability|EventActivateAbility
  (|CallParentFunction)
  (Class|NitroGameplayAnalytics|RegisterNitroAnalytics (Fortnite|GetFortPlayerControllerfromActor (Ability|GetAvatarActorfromActorInfo)) (Variables|Default|GetNitroAnalyticsSource)))
