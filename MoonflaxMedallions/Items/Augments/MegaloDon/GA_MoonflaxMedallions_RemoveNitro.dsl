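; Gameplay ability blueprint for the MegaloDon augment in MoonflaxMedallions that removes the Nitro state from its owner.
; Activation runs no logic of its own and does not end the ability.
; When the ability ends, cancelled or not, it removes owner gameplay effects that grant the tag held in the NitroStateTag variable.
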(event Ability|EventActivateAbility)

(event Ability|EventOnEndAbility
  (params (bWasCancelled :type Boolean))
  (Ability|RemoveGameplayEffectFromOwnerWithGrantedTags (Variables|Default|GetNitroStateTag)))
